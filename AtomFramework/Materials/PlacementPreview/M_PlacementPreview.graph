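# Translucent unlit material for placement previews, rendered without depth testing so that it shows through occluding geometry.
# It exposes a Color vector parameter, pale yellow-green by default, and a partial Opacity scalar, both in the Placement Preview group.
# It supports instanced static meshes and Nanite, casts no ray-traced shadows and renders in the before-depth-of-field translucency pass.
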Material /AtomFramework/Materials/PlacementPreview/M_PlacementPreview
graph {
 properties {
  bCastRayTracedShadows=false
  bDisableDepthTest=true
  bUsedWithInstancedStaticMeshes=true
  bUsedWithNanite=true
  blendMode=BLEND_Translucent
  shadingModel=MSM_Unlit
  translucencyPass=MTP_BeforeDOF
  usedShadingModels=Unlit
 }
 nodes [
  ScalarParameter() {defaultValue=0.375, group=Placement Preview, parameterName=Opacity, sortPriority=1}
  VectorParameter() {group=Placement Preview, parameterName=Color, sortPriority=0, defaultValue=<rgba 0.9354 1 0.3379 1>}
 ]
}
